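; Interactable start button for the shooting target minigame, shown with a "Start Game" prompt and a short timed interaction.
; Interacting clears AllowInteraction, which BlueprintCanInteract reads to block further use, then passes through a DoOnce to call StartGame on the referenced BPTargetMinigameManagerv2.
; A ResetOnce custom event also connects to a DoOnce leading to StartGame, and nothing in this graph sets AllowInteraction back to true.
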(event Interaction|EventBlueprintOnInteract
  (params (InteractingPawn :type Fort%20Pawn%20Object%20Reference) (InteractionBeingAttempted :type EInteractionBeingAttempted%20Enum :default "FirstInteraction"))
  (Variables|Default|SetAllowInteraction false)
  (Utilities|FlowControl|DoOnce false
    (:Completed
      (Class|BPTargetMinigameManagerv2|StartGame (Variables|Default|GetManagerRef)))))

(event Custom|ResetOnce
  (Utilities|FlowControl|DoOnce false
    (:Completed
      (Class|BPTargetMinigameManagerv2|StartGame (Variables|Default|GetManagerRef)))))

(fn BlueprintCanInteract
  (params (InteractingPawn :type Fort%20Pawn%20Object%20Reference) (InteractionBeingAttempted :type EInteractionBeingAttempted%20Enum :default "FirstInteraction") (InteractionType :type TInteractionType%20Enum :default "IT_NoInteraction"))
  (return (Variables|Default|GetAllowInteraction)))

(fn BlueprintGetInteractionString
  (params (InteractingPawn :type Fort%20Pawn%20Object%20Reference) (InteractionBeingAttempted :type EInteractionBeingAttempted%20Enum :default "FirstInteraction"))
  (return "NSLOCTEXT(\"[FC3CCC8E18BEF1197F5778F08057F57F]\", \"EF94785D4F14DC4EDF7CD3BC8E07377D\", \"Start Game\")"))

(fn BlueprintGetInteractionTime
  (params (InteractingPawn :type Fort%20Pawn%20Object%20Reference) (InteractionBeingAttempted :type EInteractionBeingAttempted%20Enum :default "FirstInteraction"))
  (return 1.0 true))

(fn ConstructionScript)
