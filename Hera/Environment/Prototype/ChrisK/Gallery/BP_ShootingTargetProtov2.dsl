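; Prototype pop-up shooting target for a target minigame that scores hits through a manager reference.
; An authority-set state enum drives deploy and retract, with one timeline played forward or in reverse to lerp the mesh rotation and advance the state when finished.
; The invulnerable actor binds its damage event on authority and accepts hits only in one state.
; A hit retracts the target, increments score on authority and plays two staggered sounds on remotes.
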(event EventBeginPlay
  ; Init
  (Networking|SwitchHasAuthority
    (:Authority
      (Variables|BuildingActor|SetbIsInvulnerable true)
      (EventDispatchers|BindEventtoOnDamaged self (delegate Custom|Hit))
      (Variables|Default|SetCurrentTargetState "NewEnumerator0")
      (Variables|Default|SetStartLoweredRelativeRotation "90.000000,90.000000,-0.000000")
      (Variables|Default|SetEndDeployedRelativeRotation "0.000000,90.000000,0.000000"))))

(event Custom|Hit
  (params (DamagedActor :type Actor%20Object%20Reference) (Damage :type Float%20%28single-precision%29) (InstigatedBy :type Controller%20Object%20Reference) (DamageCauser :type Actor%20Object%20Reference) (HitLocation :type Vector) (FHitComponent :type Primitive%20Component%20Object%20Reference) (BoneName :type Name) (Momentum :type Vector))
  ; Init
  (switch Utilities|FlowControl|Switch|SwitchonTargetMinigameTargetStatev2 (Variables|Default|GetCurrentTargetState)
    (:NewEnumerator2
      (CallFunction|TargetHit))))

(event Custom|InitTargetState
  (params (TargetState :type TargetMinigameTargetStatev2%20Enum :default "NewEnumerator0"))
  ; Set State
  (Networking|SwitchHasAuthority
    (:Authority
      (bind _output_get (Variables|Default|SetCurrentTargetState TargetState))
      (switch Utilities|FlowControl|Switch|SwitchonTargetMinigameTargetStatev2 _output_get
        (:NewEnumerator1
          (CallFunction|Deploying))
        (:NewEnumerator4
          (CallFunction|Retracting))))))

(event Custom|Deploying
  ; Deploy Target
  (bind (_direction _float) (Timeline|RaiseTarget PlayFromStart)
    (:Update
      (Transformation|SetRelativeRotation (Variables|Default|GetTargetMesh_0) (Math|Rotator|Lerp(Rotator) (Variables|Default|GetStartLoweredRelativeRotation) (Variables|Default|GetEndDeployedRelativeRotation) _float true) false false))
    (:Finished
      (switch Utilities|FlowControl|Switch|SwitchonTargetMinigameTargetStatev2 (Variables|Default|GetCurrentTargetState)
        (:NewEnumerator1
          (Variables|Default|SetCurrentTargetState "NewEnumerator2"))
        (:NewEnumerator2
          (Variables|Default|SetCurrentTargetState "NewEnumerator4"))
        (:NewEnumerator4
          (Variables|Default|SetCurrentTargetState "NewEnumerator0"))))))

(event Custom|Retracting
  ; Deploy Target
  (Timeline|RaiseTarget ReverseFromEnd))

(event Custom|TargetHit
  ; Target Hit
  (CallFunction|InitTargetState "NewEnumerator4")
  (Networking|SwitchHasAuthority
    (:Authority
      (Class|BPTargetMinigameManagerv2|IncrementScore (Variables|Default|GetManagerRef) true))
    (:Remote
      (Audio|SpawnSoundatLocation "/Game/Sounds/Fort_Ambient_NonLoop/Animals/Fort_CowMoo_Cue.Fort_CowMoo_Cue" (Transformation|GetWorldLocation (Variables|Default|GetTargetMesh_0)))
      (Utilities|FlowControl|Delay 0.33)
      (Audio|SpawnSoundatLocation "/Game/Athena/Sounds/Quests/Clays_Score_Hit_Cue.Clays_Score_Hit_Cue" (Transformation|GetWorldLocation (Variables|Default|GetTargetMesh_0))))))

(fn ConstructionScript)
