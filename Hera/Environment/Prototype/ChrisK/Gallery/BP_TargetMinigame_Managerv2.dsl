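; Manager for a prototype shooting-gallery minigame with popup and slider targets, score and countdown displays, and loot-cannon rewards.
; On authority, StartGame shuffles delays, activates sliders, starts the timer and recursively deploys random popup groups for five rounds.
; Game end fires loot for positive scores (both score tiers currently give Rare), then resets targets and button.
; OnRep functions update timer and score text; music events play only on remotes.
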(event EventBeginPlay
  ; Init
  (Networking|SwitchHasAuthority
    (:Authority
      (bind _durationseconds (CallFunction|GetGameDuration))
      (Variables|Default|SetFullGameDuration _durationseconds)
      (Variables|Default|SetCurrentGameDuration (Utilities|Text|ToText(Integer) _durationseconds))
      (Rendering|Components|TextRender|SetText (Variables|Default|GetTimerTextRender) (Variables|Default|GetCurrentGameDuration)))))

(event Custom|StartGame
  ; Start Game
  (Networking|SwitchHasAuthority
    (:Authority
      (Utilities|FlowControl|Sequence
        (:then_0
          (Utilities|FlowControl|Delay 2.0)
          (Utilities|Array|Shuffle (Variables|Default|GetAllDeployedDurationDelays))
          (Utilities|FlowControl|Sequence
            (:then_0
              (CallFunction|DeployTargets))
            (:then_1
              (for _array_element (Variables|Default|GetAllSliderTargetRefs)
                (Class|BPShootingTargetHorizontalProtov2|InitTargetState _array_element "NewEnumerator1")))
            (:then_2
              (bind _durationseconds (CallFunction|GetGameDuration))
              (Variables|Default|SetFullGameDuration _durationseconds)
              (Variables|Default|SetCurrentGameDuration (Utilities|Text|ToText(Integer) _durationseconds))
              (Rendering|Components|TextRender|SetText (Variables|Default|GetTimerTextRender) (Variables|Default|GetCurrentGameDuration))
              (CallFunction|CountDownTimer))
            (:then_3
              (CallFunction|PlayGameMusic))))
        (:then_1
          (CallFunction|PlayGameStartStingerMusic))))))

(event Custom|DeployTargets
  ; Move Targets
  (bind _selectedrandomtargets (CallFunction|PickRandomTargets))
  (Variables|Default|SetCurrentRandomTargets _selectedrandomtargets)
  (for _array_element (Variables|Default|GetCurrentRandomTargets)
    (switch Utilities|FlowControl|Switch|SwitchonTargetMinigameTargetStatev2 (Variables|/Hera/Environment/Prototype/ChrisK/Gallery/BP_ShootingTargetProtov2.BP_ShootingTargetProtov2_C|GetCurrentTargetState _array_element)
      (:NewEnumerator0
        (Class|BPShootingTargetProtov2|InitTargetState _array_element "NewEnumerator1"))))
  (Utilities|FlowControl|Delay (Utilities|Array|Get(aref) (Variables|Default|GetAllDeployedDurationDelays) (Variables|Default|GetCurrentDelayInt)))
  (for _array_element_2 (Variables|Default|GetCurrentRandomTargets)
    (switch Utilities|FlowControl|Switch|SwitchonTargetMinigameTargetStatev2 (Variables|/Hera/Environment/Prototype/ChrisK/Gallery/BP_ShootingTargetProtov2.BP_ShootingTargetProtov2_C|GetCurrentTargetState _array_element_2)
      (:NewEnumerator2
        (Class|BPShootingTargetProtov2|InitTargetState _array_element_2 "NewEnumerator4"))))
  (bind _result (Math|Integer|IncrementInt (Variables|Default|GetCurrentDelayInt))
    (:%20%20
      (if (>= _result 5)
        (Variables|Default|SetGameOver true)
        (CallFunction|PlayGameEndStingerMusic)
        (CallFunction|GiveRewards)
        (Development|PrintString "Game Over" true true "(R=0.000000,G=0.660000,B=1.000000,A=1.000000)" 1e+01)
        (else
          (Utilities|FlowControl|Delay 1.5)
          (CallFunction|DeployTargets))))))

(event Custom|CountDownTimer
  ; Countdown Clock
  (bind _returnvalue (Utilities|Time|SetTimerbyEvent (delegate Custom|DecrementTime) 1.0 true))
  (Variables|Default|SetCountdownClockHandle _returnvalue))

(event Custom|DecrementTime
  ; Countdown Clock
  (bind _result (Math|Integer|DecrementInt (Variables|Default|GetFullGameDuration))
    (:%20%20
      (bind _output_get (Variables|Default|SetCurrentGameDuration (Utilities|Text|ToText(Integer) (Math|Integer|Clamp(Integer) _result 0 999))))
      (Rendering|Components|TextRender|SetText (Variables|Default|GetTimerTextRender) _output_get))))

(event Custom|IncrementScore
  (params (Add :type Boolean :default true))
  (bind _currentgamescore (Variables|Default|GetCurrentGameScore))
  (bind _returnvalue (Utilities|String|ToString(Integer) _currentgamescore))
  ; Add Score
  (if Add
    (bind _result (Math|Integer|IncrementInt _currentgamescore)
      (:%20%20
        (Variables|Default|SetScoreText (Utilities|Text|ToText(Integer) (Math|Integer|Clamp(Integer) _result 0 9999)))
        (Variables|Default|SetCurrentGameScore (Math|Integer|Clamp(Integer) _result 0 9999))
        (Development|PrintString (Utilities|String|Append "Score = " _returnvalue) true true "(R=0.000000,G=0.660000,B=1.000000,A=1.000000)" 5.0)))
    (else
      (bind _result_2 (Math|Integer|DecrementInt _currentgamescore)
        (:%20%20
          (Variables|Default|SetScoreText (Utilities|Text|ToText(Integer) (Math|Integer|Clamp(Integer) _result_2 0 9999)))
          (Variables|Default|SetCurrentGameScore (Math|Integer|Clamp(Integer) _result_2 0 9999))
          (Development|PrintString (Utilities|String|Append "Score = " _returnvalue) true true "(R=0.000000,G=0.660000,B=1.000000,A=1.000000)" 5.0))))))

(event Custom|GiveRewards
  ; Rewards
  (Utilities|FlowControl|Delay 3.0)
  (if (> (Variables|Default|GetCurrentGameScore) 0)
    (if (>= (Variables|Default|GetCurrentGameScore) 10)
      (Variables|Default|SetLootTierReward "Loot_ApolloTreasure_Rare")
      (else
        (Variables|Default|SetLootTierReward "Loot_ApolloTreasure_Rare")))
    (Class|BGALootCannonv2|FireLootCannon (Variables|Default|GetLootCannonRef) (Variables|Default|GetLootTierReward))
    (else
      _))
  (Utilities|FlowControl|Sequence
    (:then_0
      (Utilities|FlowControl|Delay 6.0)
      (CallFunction|ResetGame))
    (:then_1
      (Utilities|FlowControl|Delay 1.0)
      (for _array_element (Variables|Default|GetAllSliderTargetRefs)
        (Class|BPShootingTargetHorizontalProtov2|ResetTargettoStart _array_element))
      (CallFunction|StopGameMusic))))

(event Custom|ResetGame
  (bind _buttonref (Variables|Default|GetButtonRef))
  ; Reset Game
  (Variables|Default|SetCurrentDelayInt 0)
  (Variables|Default|SetCurrentGameScore 0)
  (Variables|Default|SetGameOver false)
  (for _array_element (Variables|Default|GetAllPopupTargetsRef)
    (Class|BPShootingTargetProtov2|InitTargetState _array_element "NewEnumerator0"))
  (Class|BGShootingTargetButtonv2|ResetOnce _buttonref)
  (Variables|/Hera/Environment/Prototype/ChrisK/Gallery/BG_ShootingTargetButtonv2.BG_ShootingTargetButtonv2_C|SetAllowInteraction true _buttonref)
  (Variables|Default|SetScoreText "NSLOCTEXT(\"[8C83C867496EF5967C9D5D5E022922E9]\", \"CC5C9ECB42ED0C31014D5DA9462963A6\", \"0\")")
  (Utilities|Time|ClearandInvalidateTimerbyHandle (Variables|Default|GetCountdownClockHandle)))

(event Custom|PlayGameMusic
  ; Game Music
  (Networking|SwitchHasAuthority
    (:Remote
      (Audio|Components|Audio|Play (Variables|Default|GetMusic) 0.0))))

(event Custom|StopGameMusic
  ; Game Music
  (Networking|SwitchHasAuthority
    (:Remote
      (Audio|Components|Audio|FadeOut (Variables|Default|GetMusic) 2.0 0.0 "Linear" :!class /Script/Engine.AudioComponent))))

(event Custom|PlayGameStartStingerMusic
  ; Game Music
  (Networking|SwitchHasAuthority
    (:Remote
      (Audio|Components|Audio|Play (Variables|Default|GetStingerMusicStart) 0.0))))

(event Custom|PlayGameEndStingerMusic
  ; Game Music
  (Networking|SwitchHasAuthority
    (:Remote
      (Audio|Components|Audio|Play (Variables|Default|GetStingerMusicEnd) 0.0))))

(fn GetGameDuration
  (for _array_element (Variables|Default|GetAllDeployedDurationDelays)
    (LocalVariables|SetCurrentCount (+ (LocalVariables|GetCurrentCount) _array_element)))
  (return (+ (Math|Float|Truncate (LocalVariables|GetCurrentCount)) 5)))

(fn OnRep_CurrentGameDuration
  (Rendering|Components|TextRender|SetText (Variables|Default|GetTimerTextRender) (Variables|Default|GetCurrentGameDuration)))

(fn OnRep_CurrentGameScore
  (Rendering|Components|TextRender|SetText (Variables|Default|GetScoreTextRender) (Variables|Default|GetScoreText)))

(fn PickRandomMusic
  (bind _possiblesongslocal (LocalVariables|GetPossibleSongsLocal))
  (Utilities|Array|Shuffle _possiblesongslocal)
  (return (Utilities|Array|Get(acopy) _possiblesongslocal)))

(fn PickRandomTargets
  (Utilities|Array|Shuffle (Variables|Default|GetAllPopupTargetsRef))
  (Utilities|Array|Clear (Variables|Default|GetCurrentRandomTargets))
  (bind _index (Utilities|FlowControl|ForLoopwithBreak 0 (Math|Integer|Clamp(Integer) (Math|Random|RandomIntegerinRange 3 6) 3 (Utilities|Array|LastIndex (Variables|Default|GetAllPopupTargetsRef))))
    (:LoopBody
      (Utilities|Array|Add (LocalVariables|GetAllSelectedLocal) (Utilities|Array|Get(acopy) (Variables|Default|GetAllPopupTargetsRef) _index)))
    (:Completed
      (return (LocalVariables|GetAllSelectedLocal)))))

(fn ConstructionScript)
